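# Physics asset for the Pencil Kiwi male medium head's hair, giving nine hair sections secondary motion around a head capsule.
# Each section has a small MSTR root capsule constrained to the head, with one or two larger capsules below it under tight swing and twist limits.
# Most roots have free angular motion, while left and top are limited.
# The second left-hair segment attaches to its root instead of the first segment.
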PhysicsAsset /BRCosmetics/Characters/Player/Male/Medium/Heads/M_MED_Pencil_Kiwi_Head/Meshes/Hair/M_MED_Pencil_Kiwi_Hair_Physics
bodies {
 dyn_back_left_01 [{shapeType=Capsule, length=1.474, radius=3, shapeName=dyn_back_left_01_capsule, center=<1.6382 0 0>, rotation=<rot 90 0 0>}]
 dyn_back_left_02 [{shapeType=Capsule, length=0.505, radius=3, shapeName=dyn_back_left_02_capsule, center=<1.3009 0 0>, rotation=<rot 90 0 0>}]
 dyn_back_left_MSTR [{shapeType=Capsule, length=0.5, radius=0.5, shapeName=dyn_back_left_MSTR_capsule}]
 dyn_back_right_01 [{shapeType=Capsule, length=1.1446, radius=3, shapeName=dyn_back_right_01_capsule, center=<1.6455 0 0>, rotation=<rot 90 0 0>}]
 dyn_back_right_02 [{shapeType=Capsule, length=0.505, radius=3, shapeName=dyn_back_right_02_capsule, center=<1.1516 0 0>, rotation=<rot 90 0 0>}]
 dyn_back_right_MSTR [{shapeType=Capsule, length=0.5, radius=0.5, shapeName=dyn_back_right_MSTR_capsule}]
 dyn_bang_front_01 [{shapeType=Capsule, length=1.2366, radius=2.113, shapeName=dyn_bang_front_01_capsule, center=<1.8655 0 0>, rotation=<rot 90 0 0>}]
 dyn_bang_front_02 [{shapeType=Capsule, length=1.3609, radius=2.113, shapeName=dyn_bang_front_02_capsule, center=<1.8655 0 0>, rotation=<rot 90 0 0>}]
 dyn_bang_front_MSTR [{shapeType=Capsule, length=0.5, radius=0.5, shapeName=dyn_bang_front_MSTR_capsule}]
 dyn_front_bump_01 [{shapeType=Capsule, length=0.4918, radius=4.4313, shapeName=dyn_front_bump_01_capsule, center=<1.171 0 0>, rotation=<rot 90 0 0>}]
 dyn_front_bump_MSTR [{shapeType=Capsule, length=0.5, radius=0.5, shapeName=dyn_front_bump_MSTR_capsule}]
 dyn_hair_left_01 [{shapeType=Capsule, length=3.0319, radius=3, shapeName=dyn_hair_left_01_capsule, center=<2.708 0 0>, rotation=<rot 90 0 0>}]
 dyn_hair_left_02 [{shapeType=Capsule, length=1.5, radius=2, shapeName=dyn_hair_left_02_capsule, center=<1.6126 0 0>, rotation=<rot 90 0 0>}]
 dyn_hair_left_MSTR [{shapeType=Capsule, length=0.5, radius=0.5, shapeName=dyn_hair_left_MSTR_capsule}]
 dyn_hair_right_01 [{shapeType=Capsule, length=1.3701, radius=3, shapeName=dyn_hair_right_01_capsule, center=<1.6225 0 0>, rotation=<rot 90 0 0>}]
 dyn_hair_right_02 [{shapeType=Capsule, length=0.5151, radius=1.9849, shapeName=dyn_hair_right_02_capsule, center=<1.7463 0 2>}]
 dyn_hair_right_MSTR [{shapeType=Capsule, length=0.5, radius=0.5, shapeName=dyn_hair_right_MSTR_capsule}]
 dyn_hair_top_01 [{shapeType=Capsule, length=0, radius=6.7065, shapeName=dyn_hair_top_01_capsule, center=<1.748 0 4>, rotation=<rot 90 0 0>}]
 dyn_hair_top_MSTR [{shapeType=Capsule, length=0.5, radius=0.5, shapeName=dyn_hair_top_MSTR_capsule}]
 dyn_right_bang_01 [{shapeType=Capsule, length=0.3186, radius=2.113, shapeName=dyn_right_bang_01_capsule, center=<1.1998 0 0>, rotation=<rot 90 0 0>}]
 dyn_right_bang_02 [{shapeType=Capsule, length=1.4362, radius=1.5, shapeName=dyn_right_bang_02_capsule, center=<1.8655 0 0>, rotation=<rot 90 0 0>}]
 dyn_right_bang_MSTR [{shapeType=Capsule, length=0.5, radius=0.5, shapeName=dyn_right_bang_MSTR_capsule}]
 dyn_top_bang_01 [{shapeType=Capsule, length=0.4952, radius=3, shapeName=dyn_top_bang_01_capsule, center=<1.2099 0 0>, rotation=<rot 90 0 0>}]
 dyn_top_bang_02 [{shapeType=Capsule, length=1, radius=1.5, shapeName=dyn_top_bang_02_capsule, center=<1.1958 0 0>, rotation=<rot 90 0 0>}]
 dyn_top_bang_MSTR [{shapeType=Capsule, length=0.5, radius=0.5, shapeName=dyn_top_bang_MSTR_capsule}]
 head [{shapeType=Capsule, length=2.6844, radius=6.0561, center=<6.9444 2.4198 0>, rotation=<rot 90 0 0>}]
}
constraints [
 {bone1Name=dyn_hair_left_MSTR, bone2Name=head, swing1LimitDegrees=5, swing2LimitDegrees=5, twistLimitDegrees=5}
 {bone1Name=dyn_hair_left_01, bone2Name=dyn_hair_left_MSTR, swing1LimitDegrees=2, swing2LimitDegrees=15, twistLimitDegrees=2}
 {bone1Name=dyn_hair_left_02, bone2Name=dyn_hair_left_MSTR, swing1LimitDegrees=6, swing2LimitDegrees=10, twistLimitDegrees=5}
 {bone1Name=dyn_hair_right_02, bone2Name=dyn_hair_right_01, swing1LimitDegrees=5, swing2LimitDegrees=20, twistLimitDegrees=5}
 {bone1Name=dyn_hair_top_MSTR, bone2Name=head, swing1LimitDegrees=5, swing2LimitDegrees=5, twistLimitDegrees=5}
 {bone1Name=dyn_hair_top_01, bone2Name=dyn_hair_top_MSTR, swing1LimitDegrees=5, swing2LimitDegrees=5, twistLimitDegrees=5}
 {
  bone1Name=dyn_back_left_MSTR
  bone2Name=head
  swing1LimitDegrees=45
  swing1Motion=Free
  swing2LimitDegrees=45
  swing2Motion=Free
  twistLimitDegrees=45
  twistMotion=Free
 }
 {bone1Name=dyn_back_left_01, bone2Name=dyn_back_left_MSTR, swing1LimitDegrees=2, swing2LimitDegrees=2, twistLimitDegrees=2}
 {bone1Name=dyn_back_left_02, bone2Name=dyn_back_left_01, swing1LimitDegrees=6, swing2LimitDegrees=6, twistLimitDegrees=5}
 {
  bone1Name=dyn_back_right_MSTR
  bone2Name=head
  swing1LimitDegrees=45
  swing1Motion=Free
  swing2LimitDegrees=45
  swing2Motion=Free
  twistLimitDegrees=45
  twistMotion=Free
 }
 {bone1Name=dyn_back_right_01, bone2Name=dyn_back_right_MSTR, swing1LimitDegrees=2, swing2LimitDegrees=2, twistLimitDegrees=2}
 {bone1Name=dyn_back_right_02, bone2Name=dyn_back_right_01, swing1LimitDegrees=5, swing2LimitDegrees=10, twistLimitDegrees=5}
 {
  bone1Name=dyn_bang_front_MSTR
  bone2Name=head
  swing1LimitDegrees=45
  swing1Motion=Free
  swing2LimitDegrees=45
  swing2Motion=Free
  twistLimitDegrees=45
  twistMotion=Free
 }
 {bone1Name=dyn_bang_front_01, bone2Name=dyn_bang_front_MSTR, swing1LimitDegrees=10, swing2LimitDegrees=10, twistLimitDegrees=5}
 {bone1Name=dyn_bang_front_02, bone2Name=dyn_bang_front_01, swing1LimitDegrees=6, swing2LimitDegrees=6, twistLimitDegrees=5}
 {
  bone1Name=dyn_front_bump_MSTR
  bone2Name=head
  swing1LimitDegrees=45
  swing1Motion=Free
  swing2LimitDegrees=45
  swing2Motion=Free
  twistLimitDegrees=45
  twistMotion=Free
 }
 {bone1Name=dyn_front_bump_01, bone2Name=dyn_front_bump_MSTR, swing1LimitDegrees=5, swing2LimitDegrees=5, twistLimitDegrees=5}
 {bone1Name=dyn_hair_right_01, bone2Name=dyn_hair_right_MSTR, swing1LimitDegrees=10, swing2LimitDegrees=10, twistLimitDegrees=5}
 {
  bone1Name=dyn_hair_right_MSTR
  bone2Name=head
  swing1LimitDegrees=45
  swing1Motion=Free
  swing2LimitDegrees=45
  swing2Motion=Free
  twistLimitDegrees=45
  twistMotion=Free
 }
 {
  bone1Name=dyn_right_bang_MSTR
  bone2Name=head
  swing1LimitDegrees=45
  swing1Motion=Free
  swing2LimitDegrees=45
  swing2Motion=Free
  twistLimitDegrees=45
  twistMotion=Free
 }
 {bone1Name=dyn_right_bang_01, bone2Name=dyn_right_bang_MSTR, swing1LimitDegrees=2, swing2LimitDegrees=5, twistLimitDegrees=2}
 {bone1Name=dyn_right_bang_02, bone2Name=dyn_right_bang_01, swing1LimitDegrees=6, swing2LimitDegrees=10, twistLimitDegrees=5}
 {
  bone1Name=dyn_top_bang_MSTR
  bone2Name=head
  swing1LimitDegrees=45
  swing1Motion=Free
  swing2LimitDegrees=45
  swing2Motion=Free
  twistLimitDegrees=45
  twistMotion=Free
 }
 {bone1Name=dyn_top_bang_01, bone2Name=dyn_top_bang_MSTR, swing1LimitDegrees=2, swing2LimitDegrees=2, twistLimitDegrees=2}
 {bone1Name=dyn_top_bang_02, bone2Name=dyn_top_bang_01, swing1LimitDegrees=6, swing2LimitDegrees=6, twistLimitDegrees=5}
]
